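# Unlit translucent Niagara ribbon material for the smoke trail of the IgniteEgg glider.
# Particle color tinted by panning cloud noise feeds emissive through MF_FX_EmissiveAdjustments.
# Opacity multiplies particle alpha, a scrolling puffy smoke texture, fades at both ribbon ends from the second UV channel, and a FadeDist depth fade.
# The Debris reroute is a constant, and the soft-trail and streak samples are left unconnected.
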Material /BRCosmetics/Gadgets/Assets/VinderTech_GliderChute/Glider_IgniteEgg/FX/M_FX_Glider_IgniteEgg_SmokeTrail
graph {
 properties {
  bAutomaticallySetUsageInEditor=false
  bUsedWithNiagaraRibbons=true
  blendMode=BLEND_Translucent
  shadingModel=MSM_Unlit
  usedShadingModels=Unlit
 }
 nodes [
  n1 = ParticleColor()
  n2 = TextureSample(UVs=Panner(Coordinate=TextureCoordinate() {vTiling=0.25}) {constCoordinate=1, speedX=-0.2, speedY=0.2}) {texture=/Game/Effects/Fort_Effects/Textures/Smoke/T_Clouds_01}
  n3 = Multiply(A=@n1.RGB, B=Multiply(A=Add(A=@n2.G) {constB=0.5}) {constB=0.5})
  n4 = Multiply(A=NamedRerouteUsage(), B=@n3) {constB=0.2}
  MaterialFunctionCall(Emissive=Add(A=Multiply(A=@n4) {constB=0.3}, B=@n3)) {materialFunction=&MF_FX_EmissiveAdjustments}
  n5 = TextureSample(UVs=Panner(Coordinate=TextureCoordinate() {uTiling=0.25}) {speedX=-0.2}) {texture=/Game/Effects/Fort_Effects/Textures/Smoke/T_Puffy_Smoke_Trail_01}
  n6 = ComponentMask(None=TextureCoordinate() {coordinateIndex=1}) {r=true}
  n7 = Saturate(None=Multiply(A=OneMinus(None=@n6)) {constB=3})
  n8 = TextureCoordinate() {coordinateIndex=1}
  n9 = Saturate(None=Multiply(A=ComponentMask(None=@n8) {r=true}) {constB=6})
  Multiply(
   A=Multiply(A=Multiply(A=@n1.A, B=@n5.G), B=Multiply(A=@n7, B=@n9))
   B=DepthFade(FadeDistance=ScalarParameter() {defaultValue=10, parameterName=FadeDist})
  )
  NamedRerouteDeclaration(None=Constant()) {name=Debris, nodeColor=<rgba 1 0 0.8906 1>}
  TextureSample(UVs=Panner(Coordinate=TextureCoordinate() {uTiling=0.3}) {speedX=-0.1}) {samplerType=SAMPLERTYPE_Grayscale, texture=&T_Smoke_Trail_Soft}
  n10 = Panner(Coordinate=TextureCoordinate() {uTiling=4, vTiling=0.5}) {speedX=-2}
  TextureSample(UVs=OneMinus(None=MaterialFunctionCall(XY=@n10) {materialFunction=&Swizzle}.YX)) {samplerType=SAMPLERTYPE_LinearColor, texture=/Game/Effects/Fort_Effects/Textures/T_Streaks_01_M}
 ]
}
