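# Niagara system for the idle effects of the IgniteEgg glider cosmetic, with separate tuning for gameplay and front-end preview.
# A socket-attached headwind cone mesh fades with forward movement, while speedline sprites spawn in a ring, driven by forward/right inputs.
# Ribbon trail, smoke, GPU detail and spike emitters sample positions along the glider's sockets, timed by shared intro variables.
# Those four intro-timed emitters are disabled here.
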NiagaraSystem /BRCosmetics/Gadgets/Assets/VinderTech_GliderChute/Glider_IgniteEgg/FX/NS_Glider_IgniteEgg_Idle
dependencies {
 usedDataInterfaces [
  &NiagaraDataInterfaceCurve
  &NiagaraDataInterfaceVector2DCurve
  &NiagaraDataInterfaceColorCurve
  &NiagaraDataInterfaceMeshRendererInfo
  &NiagaraDataInterfaceSpriteRendererInfo
  &NiagaraDataInterfaceSkeletalMesh
  &NiagaraDataInterfaceSocketReader
 ]
 usedDynamicInputs [
  &Divide_Float
  &MakeIntFromFloat
  /Niagara/DynamicInputs/Math/Max_Int
  /Niagara/DynamicInputs/Subtract/Subtract_Integer
 ]
 usedModules [
  &self:SystemSpawnScript.SystemScriptSource.SystemScriptGraph.NiagaraNodeAssignment_0.SetVariables_#1
  &self:SystemSpawnScript.SystemScriptSource.SystemScriptGraph.NiagaraNodeAssignment_1.SetVariables_#2
  &self:SystemSpawnScript.SystemScriptSource.SystemScriptGraph.NiagaraNodeAssignment_13.SetVariables_#3
  &self:SystemSpawnScript.SystemScriptSource.SystemScriptGraph.NiagaraNodeAssignment_2.SetVariables_#4
  &self:SystemSpawnScript.SystemScriptSource.SystemScriptGraph.NiagaraNodeAssignment_9.SetVariables_#5
  &self:SystemSpawnScript.SystemScriptSource.SystemScriptGraph.NiagaraNodeAssignment_11.SetVariables_#6
  &self:SystemSpawnScript.SystemScriptSource.SystemScriptGraph.NiagaraNodeAssignment_12.SetVariables_#7
  &SystemState
  &Timeline
  &self:SystemSpawnScript.SystemScriptSource.SystemScriptGraph.NiagaraNodeAssignment_14.SetVariables_#8
  &self:SystemSpawnScript.SystemScriptSource.SystemScriptGraph.NiagaraNodeAssignment_10.SetVariables_#9
  &self:SystemSpawnScript.SystemScriptSource.SystemScriptGraph.NiagaraNodeAssignment_4.SetVariables_#10
  &self:SystemSpawnScript.SystemScriptSource.SystemScriptGraph.NiagaraNodeAssignment_5.SetVariables_#11
  &self:SystemSpawnScript.SystemScriptSource.SystemScriptGraph.NiagaraNodeAssignment_7.SetVariables_#12
  &self:SystemSpawnScript.SystemScriptSource.SystemScriptGraph.NiagaraNodeAssignment_8.SetVariables_#13
  &self:SystemSpawnScript.SystemScriptSource.SystemScriptGraph.NiagaraNodeAssignment_3.SetVariables_#14
  &SampleSkeletalMesh
  &self:SystemSpawnScript.SystemScriptSource.SystemScriptGraph.NiagaraNodeAssignment_6.SetVariables_#15
 ]
}
emitters {
 NE_HeadWindCone {
  emitterUpdateScript [
   &EmitterState(Life Cycle Mode=::NewEnumerator1, Loop Behavior=::NewEnumerator1, Loop Duration Mode=::NewEnumerator1, UseLoopDelay=false)
   &SpawnBurst_Instantaneous(Use Spawn Probability=false, UseLoopCountLimit=false)
  ]
  particleSpawnScript [
   &InitializeParticle(
    Color=<rgba 1 0.9573 0.9 1>
    Lifetime=0.35
    Mesh Scale=&FN_FrontEndVector3Modifier(Gameplay=<2 1 1>, Front End=<1 0.5 0.5>)
    Mesh Scale Mode=::NewEnumerator4
    Position Offset=<0 360 0>
    Position Offset Coordinate Space=::Local
    Sprite Size Mode=::NewEnumerator3
    UsePositionOffset=true
    Write Material Random=true
   )
   &InitialMeshOrientation(Orientation Coordinate Space=::Local, Rotation=<0 0.5 0>, Use Orientation Vector=true, Use Rotation Vector=true)
  ]
  particleUpdateScript [
   &SocketLocation(
    SamplePosOffset=true
    Sampled Position Offset=<-50 0 0>
    Socket Reader=$System.Socket Reader
    Socket Sampling Mode=::NewEnumerator3
    VelocitySpeedLimit=false
   )
   &ParticleState(Kill Particles When Lifetime Has Elapsed=false)
   &ScaleColor(
    Scale Alpha=&Multiply_Float(
     A=&Multiply_Float(
      A=&FN_FrontEndFloatModifier(Gameplay=$System.SmoothForwardDot, Front End=0.2)
      B=$System.TAlongGlider
     )
     B=5
    )
    ScaleA=true
    ScaleRGB=true
   )
  ]
  rendererClasses [&NiagaraMeshRendererProperties]
  renderers [
   {rendererClass=&NiagaraMeshRendererProperties, rendererIndex=0}
   {rendererClass=&NiagaraMeshRendererProperties, rendererIndex=1}
  ]
 }
 SideTrails {
  bEnabled=false
  emitterUpdateScript [
   &EmitterState(
    Life Cycle Mode=::NewEnumerator1
    Loop Behavior=::NewEnumerator1
    Loop Duration=&Add_Float(A=$System.IntroDuration, B=$System.IntroDelay)
    UseLoopDelay=false
   )
   &SpawnBurst_Instantaneous(Spawn Count=20, Use Spawn Probability=false, UseLoopCountLimit=false)
  ]
  particleSpawnScript [
   &PartitionParticles(
    Normalized Index In Partition Includes One=true
    Particles Per Partition=10
    Partition Mode:ENiagara_PartitionMode_v2=::NewEnumerator1
    Write Partition to Ribbon ID=::NewEnumerator0
   )
   /SetVariables(Particles.RibbonLinkOrder:NiagaraFloat=$Particles.PartitionParticles.Normalized Index In Partition) {isSetParameters=true}
   &InitializeParticle(
    Color=$System.SmokeColor
    Lifetime=$Emitter.CurrentLoopDuration
    Sprite Size Mode=::NewEnumerator3
    UsePositionOffset=false
    Write Material Random=true
   )
   &self:GetPositionAlongGlider(LastIndex=17, Socket Reader:NiagaraDataInterfaceSocketReader=$System.Socket Reader, t:NiagaraFloat=$System.TAlongGlider) {module=GetPositionAlongGlider001}
   /SetVariables(Particles.SmoothSocketOrient:Quat4f=$Output.GetPositionAlongGlider001.SampledRotation) {isSetParameters=true}
  ]
  particleUpdateScript [
   &ParticleState()
   &self:GetPositionAlongGlider(
    LastIndex=40
    Socket Reader:NiagaraDataInterfaceSocketReader=$System.Socket Reader
    t:NiagaraFloat=&Add_Float(
     A=&Multiply_Float(A=$Particles.PartitionParticles.Normalized Index In Partition, B=-0.5)
     B=$System.TAlongGlider
    )
   ) {module=GetPositionAlongGlider}
   /SetVariables(Particles.SmoothSocketOrient:Quat4f=&LerpQuaternion(
     Quaternion A:Quat4f=$Particles.SmoothSocketOrient
     Quaternion B:Quat4f=$Output.GetPositionAlongGlider.SampledRotation
     Lerp Factor:NiagaraFloat=0.1
     Enable Spherical Linear Interpolation:NiagaraBool=false
    )) {isSetParameters=true}
   /SetVariables(Particles.Position=&AddVectorToPosition(
     Position=$Output.GetPositionAlongGlider.SampledPosition
     Vector=&RotateVectorByQuaternion(
      Vector To Rotate=&MergeHelper_VectorSwitch(
       0=&Multiply_VectorByFloat(Vector=<0 0 1>, Float=$System.GliderWidth)
       1=&Multiply_VectorByFloat(Vector=<0 0 -1>, Float=$System.GliderWidth)
       2=?
       3=?
       Index=$Particles.PartitionParticles.Partition
       Num=::NewEnumerator0
      )
      Quaternion=$Particles.SmoothSocketOrient
     )
    )) {isSetParameters=true, module=SetVariables_#16}
   /SetVariables(Particles.RibbonWidth=&FloatFromCurve(
     FloatCurve=DataInterface() {propertyValues {Curve {
        defaultValue=1
        keys [
         {arriveTangent=3.5416, interpMode=RCIM_Cubic, leaveTangent=3.5416, tangentMode=RCTM_User, time=0, value=0.2}
         {interpMode=RCIM_Cubic, leaveTangent=0.0224, time=1, value=1}
        ]
       }}}
     CurveIndex=$Particles.PartitionParticles.Normalized Index In Partition
     Scale Curve=100
    )) {isSetParameters=true}
   &CameraOffset(Camera Offset Amount=50)
   &ScaleColor(
    Scale Alpha=&Multiply_Float(
     A=&FloatFromCurve(
      FloatCurve=DataInterface() {propertyValues {Curve {keys [
          <at 0 0 RCIM_Cubic>
          {arriveTangent=1.25, interpMode=RCIM_Cubic, leaveTangent=1.25, time=0.2, value=1}
          {arriveTangent=-1.25, interpMode=RCIM_Cubic, leaveTangent=-1.25, time=0.8, value=1}
          <at 1 0 RCIM_Cubic>
         ]}}}
      Scale Curve=&FN_FrontEndFloatModifier(Gameplay=1.5, Front End=0.75)
     )
     B=&Multiply_Float(
      A=1
      B=&FloatFromCurve(
       FloatCurve=DataInterface() {propertyValues {Curve {keys [<at 0 0>,<at 0.5 1>]}}}
       CurveIndex=&Subtract_Float(
        A=&Subtract_Float(A=$System.Age, B=$System.IntroDelay)
        B=&Multiply_Float(A=$Particles.PartitionParticles.Normalized Index In Partition, B=0.25)
       )
      )
     )
    )
    Scale RGB=&VectorFromFloat(Value=1.5)
    ScaleA=true
    ScaleRGB=true
   )
  ]
  rendererClasses [&NiagaraRibbonRendererProperties]
  renderers [{rendererClass=&NiagaraRibbonRendererProperties}]
 }
 Sides_-_GPU_Detail {
  bEnabled=false
  simTarget=GPUComputeSim
  emitterUpdateScript [
   &EmitterState(
    Life Cycle Mode=::NewEnumerator1
    Loop Behavior=::NewEnumerator1
    Loop Duration=&Add_Float(A=$System.IntroDuration, B=$System.IntroDelay)
    UseLoopDelay=false
   )
   &SpawnRate(SpawnRate=&Multiply_Float(A=$System.IntroSpawnWeight, B=150), Use Spawn Probability=false)
   &SpawnRate(SpawnGroup=1, SpawnRate=&Multiply_Float(A=$System.IntroSpawnWeight, B=120), Use Spawn Probability=false) {module=SpawnRate001}
  ]
  particleSpawnScript [
   /SetVariables(Particles.SpawnGroup=$Emitter.SpawnGroup, Particles.VisibilityTag=$Emitter.SpawnGroup) {isSetParameters=true, module=SetVariables_#17}
   &PartitionParticles()
   &InitializeParticle(
    Color=<rgba 0.55 0.4552 0.3368 0.9652>
    Lifetime=&MergeHelper_FloatSwitch(
     0=&RandomRangeFloat(Minimum=1, Maximum=1.5)
     1=&RandomRangeFloat(Minimum=0.1, Maximum=1)
     2=&RandomRangeFloat(Minimum=0.7, Maximum=1.2)
     3=0.5
     4=0.5
     Index=$Particles.SpawnGroup
     Num=::NewEnumerator3
    )
    Mass Mode=::NewEnumerator3
    Sprite Rotation Mode=::NewEnumerator2
    Sprite Size=&MakeVector2D(X=4, Y=&RandomRangeFloat(Minimum=2, Maximum=5))
    Sprite Size Mode=::NewEnumerator4
    UsePositionOffset=false
    Write Material Random=true
   )
   /SetVariables(Particles.LeadBone:NiagaraInt32=&RandomRangeInt(Minimum=0, Maximum=$System.LeadBone)) {isSetParameters=true, module=SetVariables_#18}
   &SocketLocation(
    Orientation Sampling=::NewEnumerator1
    SamplePosOffset=false
    Sampled Orientation Coordinate Space:ENiagaraCoordinateSpace=::World
    Socket Index=$Particles.LeadBone
    Socket Reader=$System.Socket Reader
    Socket Sampling Mode=::NewEnumerator3
    VelocitySpeedLimit=false
   )
   &SocketLocation(
    Interpolate Socket Sampling=true
    Orientation Sampling=::NewEnumerator1
    Position Sampling=::NewEnumerator1
    SamplePosOffset=false
    Sampled Orientation Coordinate Space:ENiagaraCoordinateSpace=::World
    Socket Index=&Add_Integer(A=$Particles.LeadBone, B=1)
    Socket Reader=$System.Socket Reader
    Socket Sampling Mode=::NewEnumerator3
    VelocitySpeedLimit=false
   ) {module=SocketLocation001}
   /SetVariables(Particles.DirectionBetweenBones:Vector3f=&NormalizeVector(
     Vector To Normalize=&SubtractVectorFromPosition(
      Position=$Output.SocketLocation001.SampledPosition
      Vector=Linked() {linkedVariable {name=Output.SocketLocation.SampledPosition, type=NiagaraPosition}}
     )
     Safe Normalize=false
     Safe Fallback Threshold=0
     Safe Fallback Vector=<0 0 0.001>
    )) {isSetParameters=true, module=SetVariables_#19}
   &ShapeLocation(
    MaskSpawnGroup=false
    Offset=&RotateVectorByQuaternion(
     Vector To Rotate=&MergeHelper_VectorSwitch(0=<0 0 40>, 1=<0 0 -40>, 2=?, 3=?, Index=$Particles.PartitionParticles.Partition, Num=::NewEnumerator0)
     Quaternion=$Output.SocketLocation.SampledRotation
    )
    Offset Mode=::NewEnumerator0
    Override Seed=false
    Sphere Radius=8
    UseOwnerScale=true
   )
   &AddVelocity(
    Cone Angle=&MergeHelper_FloatSwitch(0=50, 1=40, 2=20, 3=0, Index=$Particles.SpawnGroup, Num=::NewEnumerator2)
    Cone Axis=&MergeHelper_VectorSwitch(
     0=&Add_Vector(A=$Particles.DirectionBetweenBones, B=<0 0 1>)
     1=&Add_Vector(A=$Particles.DirectionBetweenBones, B=<0 0 0.1>)
     2=&Add_Vector(A=$Particles.DirectionBetweenBones, B=<0 0 0.2>)
     3=<0 0 0>
     4=<0 0 0>
     Index=$Particles.SpawnGroup
     Num=::NewEnumerator3
    )
    Distribution Along Cone Axis=0.25
    Inner Cone Angle:NiagaraFloat=&MergeHelper_FloatSwitch(0=50, 1=30, 2=20, 3=0, Index=$Particles.SpawnGroup, Num=::NewEnumerator2)
    Module.UseConeFalloff=true
    Module.UseVelDistribution=true
    Override Seed=false
    Rotation Coordinate Space=::Simulation
    Speed Falloff From Cone Axis=0.5
    Velocity Mode=::NewEnumerator2
    Velocity Speed=&MergeHelper_FloatSwitch(
     0=&RandomRangeFloat(Minimum=550, Maximum=1250)
     1=&RandomRangeFloat(Minimum=550, Maximum=850)
     2=20
     3=0
     Index=$Particles.SpawnGroup
     Num=::NewEnumerator2
    )
   )
   &SubUVAnimation(
    Random Start Frame=true
    "Sprite Renderer "=DataInterface() {propertyValues {SpriteRenderer=&self:Sides_-_GPU_Detail_0.NiagaraSpriteRendererProperties_0}}
    UseEndFrame=false
    UseStartFrame=false
   ) {enabled=false}
  ]
  particleUpdateScript [
   &ParticleState()
   &SocketLocation(
    Orientation Sampling=::NewEnumerator1
    Position Sampling=::NewEnumerator1
    SamplePosOffset=false
    Sampled Orientation Coordinate Space:ENiagaraCoordinateSpace=::World
    Socket Index=$Particles.LeadBone
    Socket Reader=$System.Socket Reader
    Socket Sampling Mode=::NewEnumerator3
    VelocitySpeedLimit=false
   ) {module=SocketLocation002}
   &SpriteRotationRate(Rotation Rate=50)
   &InheritSourceMovement(
    Applied Position Scale=&VectorFromFloat(Value=&FloatFromCurve(FloatCurve=DataInterface() {propertyValues {Curve {keys [<at 0 1 RCIM_Cubic>,<at 1 0 RCIM_Cubic>]}}}, Scale Curve=0.75))
    ApplyForce=false
    ApplyPosition=true
    ApplyPositionFalloff=false
    Source Velocity:Vector3f=$Output.SocketLocation002.SampledVelocity
   )
   &GravityForce(Gravity=&Multiply_VectorByFloat(
     Vector=<0 0 -980>
     Float=&MergeHelper_FloatSwitch(
      0=&RandomRangeFloat(Minimum=1, Maximum=1.5)
      1=0
      2=0
      3=0
      4=0
      Index=$Particles.SpawnGroup
      Num=::NewEnumerator3
     )
    ))
   &Drag(
    Drag=&MergeHelper_FloatSwitch(0=3, 1=2, 2=0, 3=0, 4=0, Index=$Particles.SpawnGroup, Num=::NewEnumerator3)
    Use Linear Drag=true
    Use Rotational Drag=false
   )
   &ScaleColor(
    Scale Alpha=&FloatFromCurve(
     FloatCurve=DataInterface() {propertyValues {Curve {
        defaultValue=1
        keys [
         {arriveTangent=9.0725, interpMode=RCIM_Cubic, leaveTangent=9.0725, tangentMode=RCTM_User, time=0, value=0}
         {arriveTangent=-2.5249, interpMode=RCIM_Cubic, leaveTangent=-2.5249, tangentMode=RCTM_User, time=0.1408, value=0.9847}
         {arriveTangent=-0.0405, interpMode=RCIM_Cubic, leaveTangent=-0.0405, tangentMode=RCTM_User, time=1, value=0}
        ]
       }}}
     Scale Curve=0.05
    )
    ScaleA=true
    ScaleRGB=true
   )
   &CurlNoiseForce(Bias Noise=false, Noise Frequency=800, Noise Strength=800, Pan Noise=false)
   &SolveForcesAndVelocity(Clamp Velocity=false, Limit Acceleration=false)
   &ScaleSpriteSizeBySpeed(Max Scale Factor=<1 8>)
   &FN_Sprite_Attributes(
    Camera Offset=10
    Color=&MultiplyLinearColor(
     Linear Color=<rgba 0.3594 0.2542 0.1761 1>
     Scale Factor=&MakeVector4FromVectorAndFloat(
      "Vector (RGB)"=<0.2 0.2 0.2>
      "Float (Alpha)"=&FloatFromCurve(FloatCurve=DataInterface() {propertyValues {Curve {
          defaultValue=1
          keys [
           {arriveTangent=9.0725, interpMode=RCIM_Cubic, leaveTangent=9.0725, tangentMode=RCTM_User, time=0, value=0}
           {arriveTangent=-2.5249, interpMode=RCIM_Cubic, leaveTangent=-2.5249, tangentMode=RCTM_User, time=0.1408, value=0.9847}
           {arriveTangent=-0.0405, interpMode=RCIM_Cubic, leaveTangent=-0.0405, tangentMode=RCTM_User, time=1, value=0}
          ]
         }}})
     )
    )
    Create Camera Offset Attribute=true
    Create Color Attribute=true
    Create Sprite Rotation Attribute=true
    Create SpriteSize Attribute=true
    SpriteRotation:NiagaraFloat=&RandomRangeFloat(Minimum=0, Maximum=360)
    SpriteSize=&Multiply_Vector2D_ByFloat(
     Float=&FloatFromCurve(FloatCurve=DataInterface() {propertyValues {Curve {
         defaultValue=0.5
         keys [
          {interpMode=RCIM_Cubic, tangentMode=RCTM_User, time=0, value=0}
          {interpMode=RCIM_Cubic, tangentMode=RCTM_User, time=0.5, value=1}
          {tangentMode=RCTM_User, time=1, value=0}
         ]
        }}})
     Vector2D=&Multiply_Vector2D_ByFloat(Float=&RandomRangeFloat(Minimum=0.5, Maximum=3), Vector2D=<3 3>)
    )
    Write to Module Namespace=true
   )
   &ScaleSpriteSize(
    Scale Factor=&MakeVector2D(X=1, Y=&FloatFromCurve(FloatCurve=DataInterface() {propertyValues {Curve {keys [<at 0 1>,<at 1 0>]}}}))
    Scale Sprite Size Mode=::NewEnumerator2
   )
  ]
  rendererClasses [&NiagaraSpriteRendererProperties]
  renderers [
   {rendererClass=&NiagaraSpriteRendererProperties}
   {rendererClass=&NiagaraSpriteRendererProperties, rendererIndex=1}
  ]
 }
 Smokes {
  bEnabled=false
  emitterSpawnScript [/SetVariables(Emitter.BeastWidth=35) {isSetParameters=true}]
  emitterUpdateScript [
   &EmitterState(
    Life Cycle Mode=::NewEnumerator1
    Loop Behavior=::NewEnumerator1
    Loop Duration=&Add_Float(A=$System.IntroDuration, B=$System.IntroDelay)
    UseLoopDelay=false
   )
   &SpawnRate(SpawnRate=&Multiply_Float(A=$System.IntroSpawnWeight, B=30), Use Spawn Probability=false)
   &SpawnRate(SpawnGroup=1, SpawnRate=&Multiply_Float(A=$System.IntroSpawnWeight, B=250), Use Spawn Probability=false) {module=SpawnRate001}
  ]
  particleSpawnScript [
   /SetVariables(Particles.SpawnGroup=$Emitter.SpawnGroup, Particles.VisibilityTag=$Emitter.SpawnGroup) {isSetParameters=true, module=SetVariables_#17}
   &PartitionParticles()
   &InitializeParticle(
    Color=<rgba 2 1.7179 1.0687 1>
    Lifetime=&MergeHelper_FloatSwitch(
     0=&RandomRangeFloat(Minimum=0.75, Maximum=1.2)
     1=&Multiply_Float(
      A=&RandomRangeFloat(Minimum=0.35, Maximum=0.8)
      B=&FN_FrontEndFloatModifier(Gameplay=1, Front End=0.5)
     )
     2=&Multiply_Float(A=&RandomRangeFloat(Minimum=0.7, Maximum=1.2), B=1)
     3=0.5
     4=0.5
     Index=$Particles.SpawnGroup
     Num=::NewEnumerator3
    )
    Mass Mode=::NewEnumerator3
    Mesh Renderer Array Visibility Mode=::NewEnumerator1
    Mesh Renderer Info=DataInterface() {propertyValues {MeshRenderer=&self:Smokes_0.NiagaraMeshRendererProperties_0}}
    Mesh Renderer Mesh Index:NiagaraInt32=&RandomRangeInt(Minimum=0, Maximum=3)
    Mesh Scale Mode=::NewEnumerator3
    Mesh Uniform Scale=&RandomRangeFloat(Minimum=0.5, Maximum=1)
    Sprite Rotation Mode=::NewEnumerator2
    Sprite Size Mode=::NewEnumerator3
    Uniform Sprite Size=&Multiply_Float(
     A=&FN_FrontEndFloatModifier(Gameplay=200, Front End=100)
     B=&FloatFromCurve(
      FloatCurve=DataInterface() {propertyValues {Curve {keys [
          {interpMode=RCIM_Cubic, tangentMode=RCTM_User, time=0, value=1}
          {arriveTangent=-1.1831, interpMode=RCIM_Cubic, leaveTangent=-1.1831, tangentMode=RCTM_User, time=1, value=0.7}
         ]}}}
      CurveIndex=$System.TAlongGlider
     )
    )
    UsePositionOffset=false
    Write Material Random=true
   )
   /SetVariables(Particles.LeadBone:NiagaraInt32=&RandomRangeInt(Minimum=$System.TrailingBone, Maximum=$System.LeadBone)) {isSetParameters=true, module=SetVariables_#18}
   &SocketLocation(
    Orientation Sampling=::NewEnumerator1
    SamplePosOffset=false
    Sampled Orientation Coordinate Space:ENiagaraCoordinateSpace=::World
    Socket Index=$Particles.LeadBone
    Socket Reader=$System.Socket Reader
    Socket Sampling Mode=::NewEnumerator3
    VelocitySpeedLimit=false
   )
   &SocketLocation(
    Interpolate Socket Sampling=true
    Orientation Sampling=::NewEnumerator1
    Position Sampling=::NewEnumerator1
    SamplePosOffset=false
    Sampled Orientation Coordinate Space:ENiagaraCoordinateSpace=::World
    Socket Index=&Add_Integer(A=$Particles.LeadBone, B=1)
    Socket Reader=$System.Socket Reader
    Socket Sampling Mode=::NewEnumerator3
    VelocitySpeedLimit=false
   ) {module=SocketLocation001}
   /SetVariables(Particles.DirectionBetweenBones:Vector3f=&NormalizeVector(
     Vector To Normalize=&SubtractVectorFromPosition(
      Position=$Output.SocketLocation001.SampledPosition
      Vector=Linked() {linkedVariable {name=Output.SocketLocation.SampledPosition, type=NiagaraPosition}}
     )
     Safe Normalize=false
     Safe Fallback Threshold=0
     Safe Fallback Vector=<0 0 0.001>
    )) {isSetParameters=true, module=SetVariables_#19}
   /SetVariables(Particles.SpawnPos:Vector3f=&RotateVectorByQuaternion(
     Vector To Rotate=&MergeHelper_VectorSwitch(
      0=&MakeVector(X=0, Y=0, Z=$Emitter.BeastWidth)
      1=&MakeVector(X=0, Y=0, Z=&Subtract_Float(A=0, B=$Emitter.BeastWidth))
      2=?
      3=?
      Index=$Particles.PartitionParticles.Partition
      Num=::NewEnumerator0
     )
     Quaternion=$Output.SocketLocation.SampledRotation
    )) {isSetParameters=true}
   &ShapeLocation(
    MaskSpawnGroup=false
    Offset=&MergeHelper_VectorSwitch(
     0=&Multiply_VectorByFloat(Vector=$Particles.SpawnPos, Float=0.75)
     1=$Particles.SpawnPos
     2=$Particles.SpawnPos
     3=$Particles.SpawnPos
     4=$Particles.SpawnPos
     Index=$Particles.SpawnGroup
     Num=::NewEnumerator3
    )
    Offset Mode=::NewEnumerator0
    Override Seed=false
    Sphere Radius=8
    UseOwnerScale=true
   )
   &AddVelocity(
    Cone Angle=&MergeHelper_FloatSwitch(0=1, 1=75, 2=20, 3=0, Index=$Particles.SpawnGroup, Num=::NewEnumerator2)
    Cone Axis=&MergeHelper_VectorSwitch(
     0=&RotateVectorByQuaternion(Vector To Rotate=<1 0 0>, Quaternion=$Output.SocketLocation.SampledRotation)
     1=&Add_Vector(A=$Particles.DirectionBetweenBones, B=<0 0 0>)
     2=&Add_Vector(A=$Particles.DirectionBetweenBones, B=<0 0 1>)
     3=?
     Index=$Particles.SpawnGroup
     Num=::NewEnumerator0
    )
    Distribution Along Cone Axis=0.25
    Module.UseConeFalloff=true
    Module.UseVelDistribution=true
    Override Seed=false
    Rotation Coordinate Space=::Simulation
    Speed Falloff From Cone Axis=0.5
    Velocity Mode=::NewEnumerator2
    Velocity Speed=&MergeHelper_FloatSwitch(
     0=&RandomRangeFloat(Minimum=500, Maximum=1250)
     1=&RandomRangeFloat(Minimum=10, Maximum=500)
     2=?
     3=?
     Index=$Particles.SpawnGroup
     Num=::NewEnumerator0
    )
   )
   &AddVelocity(
    Velocity=$Output.SocketLocation.SampledVelocity
    Velocity Speed Scale=&MergeHelper_FloatSwitch(
     0=0
     1=&RandomRangeFloat(Minimum=0, Maximum=-1)
     2=0
     3=0
     4=0
     Index=$Particles.SpawnGroup
     Num=::NewEnumerator3
    )
   ) {enabled=false, module=AddVelocity001}
  ]
  particleUpdateScript [
   &ParticleState()
   &SocketLocation(
    Orientation Sampling=::NewEnumerator1
    Position Sampling=::NewEnumerator1
    SamplePosOffset=false
    Sampled Orientation Coordinate Space:ENiagaraCoordinateSpace=::World
    Socket Index=$Particles.LeadBone
    Socket Reader=$System.Socket Reader
    Socket Sampling Mode=::NewEnumerator3
    VelocitySpeedLimit=false
   ) {module=SocketLocation002}
   &GravityForce(Gravity=&Multiply_VectorByFloat(
     Vector=<0 0 -980>
     Float=&MergeHelper_FloatSwitch(
      0=&RandomRangeFloat(Minimum=0, Maximum=0.25)
      1=1
      2=0
      3=0
      4=0
      Index=$Particles.SpawnGroup
      Num=::NewEnumerator3
     )
    ))
   &Drag(Drag=3, Use Linear Drag=true, Use Rotational Drag=false)
   &ScaleColor(
    Scale Alpha=&FloatFromCurve(FloatCurve=DataInterface() {propertyValues {Curve {keys [<at -0.0046 1.0113 RCIM_Cubic>,<at 1.0011 0.002 RCIM_Cubic>]}}})
    ScaleA=true
    ScaleRGB=true
   )
   &InheritSourceMovement(
    Applied Position Scale=&VectorFromFloat(Value=&MergeHelper_FloatSwitch(
      0=&FN_FrontEndFloatModifier(
       Gameplay=&FloatFromCurve(FloatCurve=DataInterface() {propertyValues {Curve {keys [<at 0 1 RCIM_Cubic>, {arriveTangent=0.0175, interpMode=RCIM_Cubic, leaveTangent=0.0175, tangentMode=RCTM_User, time=1, value=0}]}}})
       Front End=&FloatFromCurve(FloatCurve=DataInterface() {propertyValues {Curve {keys [<at 0 0.5 RCIM_Cubic>,<at 1 0 RCIM_Cubic>]}}}, Scale Curve=1.5)
      )
      1=&FN_FrontEndFloatModifier(
       Gameplay=1
       Front End=&FloatFromCurve(FloatCurve=DataInterface() {propertyValues {Curve {keys [<at 0 0.5 RCIM_Cubic>,<at 1 0 RCIM_Cubic>]}}}, Scale Curve=1.5)
      )
      2=?
      3=?
      Index=$Particles.SpawnGroup
      Num=::NewEnumerator0
     ))
    ApplyForce=false
    ApplyPosition=true
    ApplyPositionFalloff=false
    Source Velocity:Vector3f=$Output.SocketLocation002.SampledVelocity
   )
   &SolveForcesAndVelocity(Clamp Velocity=false, Limit Acceleration=false)
   &FN_Sprite_Attributes(
    Camera Offset=20
    Color=&MultiplyLinearColor(
     Linear Color=$System.SmokeColor
     Scale Factor=&MakeVector4FromVectorAndFloat(
      "Vector (RGB)"=<0.4 0.4 0.4>
      "Float (Alpha)"=&FloatFromCurve(
       FloatCurve=DataInterface() {propertyValues {Curve {
          defaultValue=0
          keys [
           {interpMode=RCIM_Cubic, tangentMode=RCTM_User, time=0, value=0}
           {interpMode=RCIM_Cubic, tangentMode=RCTM_User, time=0.5, value=1}
           {tangentMode=RCTM_User, time=1, value=0}
          ]
         }}}
       Scale Curve=&FN_FrontEndFloatModifier(Gameplay=0.65, Front End=0.4)
      )
     )
    )
    Create Camera Offset Attribute=true
    Create Color Attribute=true
    Create Sprite Rotation Attribute=true
    Create SpriteSize Attribute=true
    SpriteRotation:NiagaraFloat=&RandomRangeFloat(Minimum=0, Maximum=360)
    SpriteSize=&Multiply_Vector2D_ByFloat(
     Float=&FloatFromCurve(
      FloatCurve=DataInterface() {propertyValues {Curve {
         defaultValue=1
         keys [
          {arriveTangent=4.1785, interpMode=RCIM_Cubic, leaveTangent=1.6747, tangentMode=RCTM_Break, time=0, value=0.5}
          {arriveTangent=0.3538, interpMode=RCIM_Cubic, leaveTangent=0.0224, tangentMode=RCTM_Break, time=1, value=1.3}
         ]
        }}}
      Scale Curve=0.85
     )
     Vector2D=$Particles.Initial.SpriteSize
    )
    Write to Module Namespace=true
   )
   &FN_Mesh_Attributes(
    Create Scale Attribute=true
    Scale=&Multiply_VectorByFloat(
     Vector=$Particles.Initial.Scale
     Float=&FloatFromCurve(FloatCurve=DataInterface() {propertyValues {Curve {
         defaultValue=1
         keys [
          {arriveTangent=9.0725, interpMode=RCIM_Cubic, time=0, value=0}
          <at 0.3 0.8 RCIM_Cubic>
          {interpMode=RCIM_Cubic, leaveTangent=-0.0405, time=1, value=0}
         ]
        }}})
    )
    Write to Module Namespace=true
   )
   /SetVariables(Particles.DebrisParams:Vector4f=<0 0.3 1 0>) {isSetParameters=true}
   &UpdateMeshOrientation(
    Rotation Coordinate Space=::NewEnumerator1
    Rotation Rate=&RandomRangeFloat(Minimum=1, Maximum=3)
    Rotation Vector=/Niagara/DynamicInputs/Vectors/CrossProduct(Vector A:Vector3f=$Particles.DirectionBetweenBones, Vector B:Vector3f=<-1 0 0>, Normalize Incoming Vectors:NiagaraBool=true)
   )
   &FN_Sprite_Attributes(
    Camera Offset=50
    Color=&MultiplyLinearColor(
     Linear Color=<rgba 0.3594 0.2542 0.1761 1>
     Scale Factor=&MakeVector4FromVectorAndFloat(
      "Vector (RGB)"=<2 2 2>
      "Float (Alpha)"=&FloatFromCurve(FloatCurve=DataInterface() {propertyValues {Curve {
          defaultValue=1
          keys [
           {arriveTangent=9.0725, interpMode=RCIM_Cubic, leaveTangent=9.0725, tangentMode=RCTM_User, time=0, value=0}
           {arriveTangent=-2.5249, interpMode=RCIM_Cubic, leaveTangent=-2.5249, tangentMode=RCTM_User, time=0.1408, value=0.9847}
           {arriveTangent=-0.0405, interpMode=RCIM_Cubic, leaveTangent=-0.0405, tangentMode=RCTM_User, time=1, value=0}
          ]
         }}})
     )
    )
    Create Camera Offset Attribute=true
    Create Color Attribute=true
    Create SpriteSize Attribute=true
    SpriteSize=&Multiply_Vector2D_ByFloat(
     Float=&RandomRangeFloat(Minimum=0.075, Maximum=0.2)
     Vector2D=&Multiply_Vector2D(
      A=<153.3176 500>
      B=&MakeVector2D(
       X=1
       Y=&FloatFromCurve(FloatCurve=DataInterface() {propertyValues {Curve {
           defaultValue=1
           keys [
            {arriveTangent=4.1785, interpMode=RCIM_Cubic, leaveTangent=4.1785, tangentMode=RCTM_User, time=0, value=0}
            {arriveTangent=1.1179, interpMode=RCIM_Cubic, leaveTangent=1.1179, tangentMode=RCTM_User, time=0.4, value=1.5}
            {arriveTangent=0.0224, interpMode=RCIM_Cubic, leaveTangent=0.0224, tangentMode=RCTM_User, time=1, value=1.75}
           ]
          }}})
      )
     )
    )
    Write to Module Namespace=true
   ) {module=FN_Sprite_Attributes001}
   /SetVariables(Particles.SpikesDynParams:Vector4f=&MakeVector4(X=0, Y=0, Z=0, W=0)) {isSetParameters=true}
   &DynamicMaterialParameters(
    Index 0 Param 1=&Add_Float(
     A=&FloatFromCurve(FloatCurve=DataInterface() {propertyValues {Curve {keys [<at 0 0>,<at 1 1>]}}}, Scale Curve=0.2)
     B=&Multiply_Float(A=$Particles.Initial.MaterialRandom, B=0.5)
    )
    Index 0 Param 2=2
    Index 0 Param 3=&FloatFromCurve(
     FloatCurve=DataInterface() {propertyValues {Curve {keys [
         {arriveTangent=-0.0944, interpMode=RCIM_Cubic, time=0, value=0}
         {arriveTangent=2.5569, interpMode=RCIM_Cubic, leaveTangent=2.5569, time=0.0519, value=2.3053}
         <at 0.5008 1.2804>
        ]}}}
     Scale Curve=0
    )
    Index 0 Param 4=1
    Index 1 Param 1=1
    Index 1 Param 2=1
    Index 1 Param 3=1
    Index 1 Param 4=1
    Index 2 Param 1=&FloatFromCurve(
     FloatCurve=DataInterface() {propertyValues {Curve {defaultValue=1, keys [{tangentMode=RCTM_User, time=0, value=0}, {tangentMode=RCTM_User, time=1, value=1}]}}}
     Scale Curve=3
    )
    Index 2 Param 3=0.5
    Index 2 Param 4=2
    Module.Param0WriteEnabled=true
    Module.Param0WriteEnabled1=true
    Module.Param0WriteEnabled2=true
    Module.Param1WriteEnabled=true
    Module.Param1WriteEnabled1=true
    Module.Param1WriteEnabled2=true
    Module.Param2WriteEnabled=true
    Module.Param2WriteEnabled1=true
    Module.Param2WriteEnabled2=true
    Module.Param3WriteEnabled=true
    Module.Param3WriteEnabled1=true
    Module.Param3WriteEnabled2=true
    Write Parameter Index 1=true
    Write Parameter Index 2=true
   ) {module=DynamicMaterialParameters001}
  ]
  rendererClasses [&NiagaraSpriteRendererProperties, &NiagaraMeshRendererProperties]
  renderers [
   {rendererClass=&NiagaraSpriteRendererProperties}
   {rendererClass=&NiagaraMeshRendererProperties, rendererIndex=1}
  ]
 }
 Speedlines001 {
  emitterUpdateScript [
   &EmitterState()
   &SpawnRate(
    Spawn Probability=&MakeCustomFloatFromBool(
     Bool=&SetBoolByFloatComparison(A=$User.ForwardDot, B=0.05)
     True Float=1
     False Float=&FN_FrontEndFloatModifier(Gameplay=0, Front End=1)
    )
    SpawnRate=&MakeCustomFloatFromBool(Bool=$User.bIsFrontEndPreview, True Float=50, False Float=90)
    Use Spawn Probability=true
   ) {module=SpawnRate001}
  ]
  particleSpawnScript [
   &InitializeParticle(
    Color=&MakeLinearColorFromVectorAndFloat(
     "Vector (RGB)"=<1 1 1>
     "Float (Alpha)"=&Multiply_Float(
      A=&UniformRangedFloat(Minimum=1, Maximum=1)
      B=&MakeCustomFloatFromBool(
       Bool=$User.bIsFrontEndPreview
       True Float=&FloatFromCurve(
        FloatCurve=DataInterface() {propertyValues {Curve {keys [{tangentMode=RCTM_User, time=0, value=0}, {tangentMode=RCTM_User, time=2, value=1}]}}}
        CurveIndex=$Emitter.Age
        Scale Curve=0.4
       )
       False Float=1
      )
     )
    )
    Lifetime Max=0.2
    Lifetime Min=0.15
    Lifetime Mode=::NewEnumerator1
    Position Offset=&FN_FrontEndVector3Modifier(Gameplay=<0 -50 0>, Front End=<0 50 0>)
    Sprite Size Max=<3 10>
    Sprite Size Min=<2 2>
    Sprite Size Mode=::NewEnumerator2
    UsePositionOffset=true
    Write Material Random=true
   )
   &AddVelocity(
    Module.Use Falloff Distance=false
    Override Seed=false
    Velocity Mode=::NewEnumerator1
    Velocity Speed=&FN_FrontEndFloatModifier(Gameplay=-2500, Front End=2500)
   ) {module=AddVelocity001}
   &TorusLocation(
    Align Torus With Axis=true
    Coordinate Space:ENiagaraCoordinateSpace=::Simulation
    Handle Radius=0
    Large Radius:NiagaraFloat=&FN_FrontEndFloatModifier(
     Gameplay=&RandomRangeFloat(Minimum=450, Maximum=500)
     Front End=&RandomRangeFloat(Minimum=350, Maximum=400)
    )
    MaskSpawnGroup=false
    Module.Override Randomness=false
    Offset=<0 33 0>
    Torus Axis=<0 1 0>
    Torus Origin:NiagaraPosition=&AddVectorToPosition(
     Position=$Particles.Position
     Vector=&FN_FrontEndVector3Modifier(Gameplay=<0 -500 0>, Front End=<0 -500 0>)
    )
   )
   &AddVelocity(
    Velocity=&Add_Vector(
     A=&RandomRangeVector(Minimum=<0 300 0>, Maximum=<0 700 1>)
     B=&Add_Vector(
      A=&Multiply_VectorByFloat(Vector=<-555 0 0>, Float=$User.RightDot)
      B=&Multiply_VectorByFloat(Vector=<0 0 222>, Float=$User.ForwardDot)
     )
    )
    Velocity Speed Scale=&FN_FrontEndFloatModifier(Gameplay=1, Front End=-1)
   )
  ]
  particleUpdateScript [
   &SolveForcesAndVelocity(Clamp Velocity=false, Limit Acceleration=false)
   &ParticleState()
   &ScaleSpriteSize(Scale Factor=&Vector2DFromCurve(
     Vector2Curve=DataInterface() {propertyValues {XCurve {keys [<at 0 0.5 RCIM_Cubic>,<at 0.5 1 RCIM_Cubic>]}, YCurve {keys [<at 0 0 RCIM_Cubic>,<at 0.5 1 RCIM_Cubic>]}}}
     ScaleCurve=&Vector2DFromFloat(Value=1)
    ))
   &ScaleSpriteSizeBySpeed(Max Scale Factor=<1 50>, Min Scale Factor=<0.01 0.05>, Velocity Threshold=1)
   &ScaleColor(
    Scale Alpha=&FloatFromCurve(
     FloatCurve=DataInterface() {propertyValues {Curve {keys [
         <at 0 0 RCIM_Cubic>
         <at 0.3759 1.0207>
         {arriveTangent=-2.221, interpMode=RCIM_Cubic, leaveTangent=-2.221, tangentMode=RCTM_User, time=1, value=0}
        ]}}}
     Scale Curve=&MakeCustomFloatFromBool(
      Bool=$User.bIsFrontEndPreview
      True Float=&MakeCustomFloatFromBool(Bool=$User.bIsFullyDeployed, True Float=0.5, False Float=0)
      False Float=&Lerp_Float(A=0, B=0.5, Alpha=&FN_FrontEndFloatModifier(Gameplay=$System.Timeline.Time, Front End=0.5))
     )
    )
    Scale RGB=&VectorFromFloat(Value=&FN_FrontEndFloatModifier(Gameplay=0.5, Front End=0.33))
    ScaleA=true
    ScaleRGB=true
   )
  ]
  rendererClasses [&NiagaraSpriteRendererProperties]
  renderers [{rendererClass=&NiagaraSpriteRendererProperties}]
 }
 Spikes {
  bEnabled=false
  emitterUpdateScript [
   &EmitterState(
    Life Cycle Mode=::NewEnumerator1
    Loop Behavior=::NewEnumerator1
    Loop Duration=&Add_Float(A=$System.IntroDuration, B=$System.IntroDelay)
    UseLoopDelay=false
   )
   &SpawnRate(SpawnRate=&Multiply_Float(A=40, B=$System.IntroSpawnWeight), Use Spawn Probability=false)
  ]
  particleSpawnScript [
   &PartitionParticles(Partition Mode:ENiagara_PartitionMode_v2=::NewEnumerator5, Write Partition to Ribbon ID=::NewEnumerator0)
   &InitializeParticle(
    Color=$System.SmokeColor
    Lifetime=&RandomRangeFloat(Minimum=0.5, Maximum=0.8)
    Sprite Size=&Multiply_Vector2D_ByFloat(Float=&FN_FrontEndFloatModifier(Gameplay=1, Front End=0.5), Vector2D=<60 60>)
    Sprite Size Mode=::NewEnumerator4
    UsePositionOffset=false
    Write Material Random=true
   )
   &SubUVAnimation(End Frame=3, SubUV Animation Mode:ENiagara_SubUVLookupMode=::NewEnumerator2)
   &CameraOffset(Camera Offset Amount=&FN_FrontEndFloatModifier(Gameplay=0, Front End=50)) {module=CameraOffset001}
   &self:GetPositionAlongGlider(LastIndex=17, Socket Reader:NiagaraDataInterfaceSocketReader=$System.Socket Reader, t:NiagaraFloat=$System.TAlongGlider) {module=GetPositionAlongGlider001}
   /SetVariables(Particles.SmoothSocketOrient:Quat4f=$Output.GetPositionAlongGlider001.SampledRotation) {isSetParameters=true}
  ]
  particleUpdateScript [
   &SolveForcesAndVelocity(Clamp Velocity=false, Limit Acceleration=false)
   &ParticleState()
   &self:GetPositionAlongGlider(
    LastIndex=40
    OffsetInSocketSpace:Vector3f=&MergeHelper_VectorSwitch(
     0=&Multiply_VectorByFloat(Vector=<0 0 0>, Float=$System.GliderWidth)
     1=&Multiply_VectorByFloat(Vector=<0 0 0>, Float=$System.GliderWidth)
     2=<0 0 0>
     3=<0 0 0>
     4=<0 0 0>
     Index=$Particles.PartitionParticles.Partition
     Num=::NewEnumerator3
    )
    Socket Reader:NiagaraDataInterfaceSocketReader=$System.Socket Reader
    t:NiagaraFloat=&Add_Float(A=$System.TAlongGlider, B=&RandomRangeFloat(Minimum=0, Maximum=0.05))
   ) {module=GetPositionAlongGlider}
   /SetVariables(Particles.SmoothSocketOrient:Quat4f=&LerpQuaternion(
     Quaternion A:Quat4f=$Particles.SmoothSocketOrient
     Quaternion B:Quat4f=$Output.GetPositionAlongGlider.SampledRotation
     Lerp Factor:NiagaraFloat=0.05
     Enable Spherical Linear Interpolation:NiagaraBool=false
    )) {isSetParameters=true}
   /SetVariables(Particles.Position=&AddVectorToPosition(
     Position=$Output.GetPositionAlongGlider.SampledPosition
     Vector=&RotateVectorByQuaternion(
      Vector To Rotate=&MergeHelper_VectorSwitch(
       0=&Multiply_VectorByFloat(Vector=<0 0 0.6>, Float=$System.GliderWidth)
       1=&Multiply_VectorByFloat(Vector=<0 0 -0.6>, Float=$System.GliderWidth)
       2=?
       3=?
       Index=$Particles.PartitionParticles.Partition
       Num=::NewEnumerator0
      )
      Quaternion=$Particles.SmoothSocketOrient
     )
    )) {isSetParameters=true, module=SetVariables_#16}
   &CameraOffset()
   &UpdateMeshOrientation(
    Facing Direction=&RotateVectorByQuaternion(
     Vector To Rotate=&MakeVector(
      X=-1
      Y=&RandomRangeFloat(Minimum=-1, Maximum=1)
      Z=&MergeHelper_FloatSwitch(
       0=&RandomRangeFloat(Minimum=0, Maximum=1)
       1=&RandomRangeFloat(Minimum=0, Maximum=-1)
       2=?
       3=?
       Index=$Particles.PartitionParticles.Partition
       Num=::NewEnumerator0
      )
     )
     Quaternion=$Particles.SmoothSocketOrient
    )
    Facing Interpolation=true
    Facing Interpolation Rate=5
    Orientation Method=::NewEnumerator1
    Override Quat=false
    Stabilization Mode:ENiagara_OrientVectorCount=::NewEnumerator0
    TransformFacing=false
   )
   &DynamicMaterialParameters(
    Index 0 Param 1=200
    Index 0 Param 4=&FloatFromCurve(FloatCurve=DataInterface() {propertyValues {Curve {keys [{interpMode=RCIM_Cubic, tangentMode=RCTM_User, time=0, value=0}, {interpMode=RCIM_Cubic, tangentMode=RCTM_User, time=1, value=1}]}}})
    Index 1 Param 1=&FloatFromCurve(
     FloatCurve=DataInterface() {propertyValues {Curve {keys [
         {interpMode=RCIM_Cubic, tangentMode=RCTM_User, time=0, value=0}
         {arriveTangent=0.2407, interpMode=RCIM_Cubic, leaveTangent=0.2407, tangentMode=RCTM_User, time=1, value=1}
        ]}}}
     Scale Curve=&RandomRangeFloat(Minimum=0.1, Maximum=1)
    )
    Index 1 Param 3=0.5
    Index 1 Param 4=&FloatFromCurve(FloatCurve=DataInterface() {propertyValues {Curve {keys [<at 0 1>,<at 1 0>]}}}, Scale Curve=-0.3)
    Module.Index0Param0Write=true
    Module.Index0Param1Write=false
    Module.Index0Param2Write=false
    Module.Index0Param3Write=true
    Module.Index1Param0Write=true
    Module.Index1Param1Write=true
    Module.Index1Param2Write=true
    Module.Index1Param3Write=true
    Write Parameter Index 1=true
   )
   &AlignSpriteToMeshOrientation(
    Mesh Orientation Relative Sprite Alignment Vector=<1 0 0>
    Mesh Orientation Relative Sprite Facing Vector:Vector3f=<0 0 1>
    SetAlignment=true
    SetFacing=true
   )
   &ScaleSpriteSize(Scale Factor=&Vector2DFromCurve(
     Vector2Curve=DataInterface() {propertyValues {
       XCurve {keys [
         {arriveTangent=0.4813, interpMode=RCIM_Cubic, leaveTangent=0.4813, tangentMode=RCTM_User, time=0, value=0.625}
         {arriveTangent=1.25, interpMode=RCIM_Cubic, leaveTangent=1.25, time=0.4, value=0.9716}
         {arriveTangent=2.6138, interpMode=RCIM_Cubic, leaveTangent=2.6138, tangentMode=RCTM_User, time=0.9, value=1.75}
        ]}
       YCurve {keys [
         {arriveTangent=9.4573, interpMode=RCIM_Cubic, leaveTangent=9.4573, tangentMode=RCTM_User, time=0, value=0}
         {arriveTangent=0.8193, interpMode=RCIM_Cubic, leaveTangent=0.8193, tangentMode=RCTM_User, time=0.2, value=1.5019}
         <at 1 1.7992 RCIM_Cubic>
        ]}
      }}
     ScaleCurve=<1 2>
    ))
   &ScaleSpriteSize(
    Scale Factor=&Multiply_Vector2D_ByFloat(
     Float=&FloatFromCurve(
      FloatCurve=DataInterface() {propertyValues {Curve {keys [
          {interpMode=RCIM_Cubic, tangentMode=RCTM_User, time=0, value=1}
          {arriveTangent=-1.864, interpMode=RCIM_Cubic, leaveTangent=-1.864, tangentMode=RCTM_User, time=1, value=0.6}
         ]}}}
      CurveIndex=$System.TAlongGlider
     )
     Vector2D=&MakeVector2D(
      X=&FloatFromCurve(
       FloatCurve=DataInterface() {propertyValues {Curve {keys [
           {arriveTangent=-0.2333, interpMode=RCIM_Cubic, leaveTangent=-0.2333, tangentMode=RCTM_User, time=0, value=2}
           {arriveTangent=-3.2303, interpMode=RCIM_Cubic, tangentMode=RCTM_Break, time=1, value=1}
          ]}}}
       Scale Curve=0.9
      )
      Y=&FloatFromCurve(FloatCurve=DataInterface() {propertyValues {Curve {keys [
           {arriveTangent=2.5653, interpMode=RCIM_Cubic, leaveTangent=2.5653, tangentMode=RCTM_User, time=0, value=0.75}
           {interpMode=RCIM_Cubic, leaveTangent=5.1925, time=1, value=2.5}
          ]}}})
     )
    )
    Scale Sprite Size Mode=::NewEnumerator2
   ) {module=ScaleSpriteSize001}
   &ScaleColor(
    Scale Alpha=&FloatFromCurve(
     FloatCurve=DataInterface() {propertyValues {Curve {
        defaultValue=1
        keys [
         {arriveTangent=9.0725, interpMode=RCIM_Cubic, leaveTangent=9.0725, tangentMode=RCTM_User, time=0, value=0}
         {arriveTangent=-2.5249, interpMode=RCIM_Cubic, leaveTangent=-2.5249, tangentMode=RCTM_User, time=0.1408, value=0.9847}
         {arriveTangent=-0.0405, interpMode=RCIM_Cubic, leaveTangent=-0.0405, tangentMode=RCTM_User, time=1, value=0}
        ]
       }}}
     Scale Curve=&FN_FrontEndFloatModifier(Gameplay=0.8, Front End=0.5)
    )
    Scale RGB=&MakeVectorFromLinearColorRGB(Color=&ColorFromCurve(ColorCurve=DataInterface() {propertyValues {
        AlphaCurve {keys [<at 0 1 RCIM_Cubic>]}
        BlueCurve {keys [<at 0 2.5 RCIM_Cubic>,<at 0.4667 1 RCIM_Cubic>]}
        GreenCurve {keys [<at 0 3.8288 RCIM_Cubic>,<at 0.4667 1 RCIM_Cubic>]}
        RedCurve {keys [<at 0 5 RCIM_Cubic>,<at 0.4667 1 RCIM_Cubic>]}
       }}))
    ScaleA=true
    ScaleRGB=true
   ) {module=ScaleColor001}
  ]
  rendererClasses [&NiagaraSpriteRendererProperties]
  renderers [{rendererClass=&NiagaraSpriteRendererProperties}]
 }
}
summary {
 systemName=NS_Glider_IgniteEgg_Idle
 emitters [
  {emitterName=Speedlines001, rendererClasses [&NiagaraSpriteRendererProperties]}
  {
   bEnabled=false
   emitterName=Smokes
   rendererClasses [&NiagaraSpriteRendererProperties, &NiagaraMeshRendererProperties]
  }
  {bEnabled=false, emitterName=Sides_-_GPU_Detail, simTarget=GPUComputeSim, rendererClasses [&NiagaraSpriteRendererProperties]}
  {bEnabled=false, emitterName=SideTrails, rendererClasses [&NiagaraRibbonRendererProperties]}
  {bEnabled=false, emitterName=Spikes, rendererClasses [&NiagaraSpriteRendererProperties]}
  {emitterName=NE_HeadWindCone, rendererClasses [&NiagaraMeshRendererProperties]}
 ]
 userVariables [
  User.bIsFrontEnd {type=NiagaraBool}
  User.bIsFrontEndPreview {type=NiagaraBool}
  User.bIsFullyDeployed {type=NiagaraBool}
  User.FakeForwardDot {type=NiagaraFloat, defaultValue=0, description=""}
  User.ForwardDot {type=NiagaraFloat, defaultValue=1, description=""}
  User.RightDot {type=NiagaraFloat, defaultValue=0, description=""}
  User.SystemDelay {type=NiagaraFloat, defaultValue=1, description=""}
 ]
}
