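; Anim notify state for the IgniteEgg glider that keeps the skeletal mesh's overlay material in step with its base material's effect mask.
; On begin, it wraps the overlay in a dynamic material instance if it is not one already.
; Each tick, it copies the EffectMask scalar from the mesh's slot 0 dynamic material to the overlay, or creates and assigns the overlay instance if the cast fails.
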(fn Received_NotifyBegin
  (params (MeshComp :type Skeletal%20Mesh%20Component%20Object%20Reference) (Animation :type Anim%20Sequence%20Base%20Object%20Reference) (TotalDuration :type Float%20%28single-precision%29) (EventReference :type Anim%20Notify%20Event%20Reference%20Structure%20%28by%20ref%29))
  (bind _overlaymaterial (Variables|/Script/Engine.MeshComponent|GetOverlayMaterial MeshComp))
  (Utilities|Casting|CastToMaterialInstanceDynamic _overlaymaterial
    (:then
      _)
    (:CastFailed
      (bind _returnvalue (Rendering|Material|CreateDynamicMaterialInstance _overlaymaterial _overlaymaterial "None" "None" :!class /Script/Engine.KismetMaterialLibrary))
      (Rendering|Material|SetOverlayMaterial MeshComp _returnvalue false 0 :!class /Script/Engine.MeshComponent)))
  (return false))

(fn Received_NotifyTick
  (params (MeshComp :type Skeletal%20Mesh%20Component%20Object%20Reference) (Animation :type Anim%20Sequence%20Base%20Object%20Reference) (FrameDeltaTime :type Float%20%28single-precision%29) (EventReference :type Anim%20Notify%20Event%20Reference%20Structure%20%28by%20ref%29))
  (bind _returnvalue (Rendering|Material|GetOverlayMaterial MeshComp false 0 :!class /Script/Engine.MeshComponent))
  (bind _asmaterial_instance_dynamic (Utilities|Casting|CastToMaterialInstanceDynamic _returnvalue)
    (:then
      (bind _returnvalue_2 (Rendering|Material|GetScalarParameterValue (Class|BTechArtGlobalLibrary|GetDynamicMaterialInstance MeshComp 0 MeshComp) "EffectMask" :!class /Script/Engine.MaterialInstanceDynamic))
      (Rendering|Material|SetScalarParameterValue _asmaterial_instance_dynamic "EffectMask" _returnvalue_2 :!class /Script/Engine.MaterialInstanceDynamic)
      (return false))
    (:CastFailed
      (bind _returnvalue_3 (Rendering|Material|CreateDynamicMaterialInstance MeshComp _returnvalue "None" "None" :!class /Script/Engine.KismetMaterialLibrary))
      (Rendering|Material|SetOverlayMaterial MeshComp _returnvalue_3 false 0 :!class /Script/Engine.MeshComponent))))
