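# Material function for the IgniteEgg glider that builds animated smoke noise and a dissolve-style opacity mask.
# Two smoke texture samples pan at different speeds over pre-skinned local XY, combined by max into DustNoise.
# A local-Y gradient, shifted by the EffectMask parameter, forms a sweep mask that thresholds the noise into MaskedNoise.
# DitherTemporalAA turns that into OpacityMask, and vertex interpolators move coordinate math to the vertex shader.
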MaterialFunction /BRCosmetics/Gadgets/Assets/VinderTech_GliderChute/Glider_IgniteEgg/Materials/MF_IgniteEgg_Glider
graph {nodes [
  n1 = ComponentMask(None=LocalPosition() {localOrigin=InstancePreSkinning}.XYZ) {g=true, r=true}
  n2 = VertexInterpolator(VS=Multiply(A=Multiply(A=@n1) {constB=0.01}) {constB=0.4})
  n3 = TextureSample(UVs=Panner(Coordinate=@n2.PS) {speedX=0.05, speedY=-0.5}) {texture=/Game/Effects/Fort_Effects/Textures/Tile/T_Inky_Smoke_Tile}
  n4 = TextureSample(UVs=Panner(Coordinate=@n2.PS) {speedX=0.05, speedY=-0.3}) {texture=&T_Inky_Smoke_Tile}
  FunctionOutput(None=NamedRerouteDeclaration(None=Max(A=@n3.R, B=@n4.R)) {name=Noise, nodeColor=<rgba 0.4062 1 0 1>}) {outputName=DustNoise}
  n5 = NamedRerouteUsage()
  n6 = NamedRerouteDeclaration(None=SmoothStep(Min=OneMinus(None=@n5), Value=NamedRerouteUsage())) {name=MaskedNoise, nodeColor=<rgba 0 0.8828 1 1>}
  FunctionOutput(None=MaterialFunctionCall(Alpha Threshold=@n6) {materialFunction=&DitherTemporalAA}.Result) {outputName=OpacityMask}
  FunctionOutput(None=@n6) {outputName=MaskedNoise}
  n7 = LocalPosition() {localOrigin=InstancePreSkinning}
  n8 = VertexInterpolator(VS=Multiply(A=ComponentMask(None=@n7.XYZ) {g=true}) {constB=0.01})
  n9 = ScalarParameter() {defaultValue=-1.7, parameterName=MaskPosAdjust}
  n10 = Saturate(None=Divide(A=Subtract(A=@n8.PS, B=@n9), B=ScalarParameter() {defaultValue=8, parameterName=MaskSmoothness}))
  n11 = OneMinus(None=ScalarParameter() {defaultValue=1, parameterName=EffectMask})
  n12 = Saturate(None=Add(A=Subtract(A=@n10), B=LinearInterpolate(Alpha=@n11) {constB=2}))
  NamedRerouteDeclaration(None=OneMinus(None=SmoothStep(Value=@n12) {constMax=0.65, constMin=0.35})) {name=EffectMask, nodeColor=<rgba 0.5 1 0 1>}
 ]}
