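# Masked, two-sided material for the spline-thickened parts of the IgniteEgg glider cosmetic.
# MF_splineThickenM supplies World Position Offset and a non-tangent-space normal, with width driven by vertex color red times a parameter.
# A texture mask blends black to dark blue-gray, MF_IgniteEgg_Glider supplies the opacity mask, and attributes feed the accessories tech-art override.
# An unconnected hand-built camera-facing thicken and normal network remains in the graph.
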Material /BRCosmetics/Gadgets/Assets/VinderTech_GliderChute/Glider_IgniteEgg/Materials/M_Glider_IgniteEgg_SplineThicken
graph {
 properties {
  bTangentSpaceNormal=false
  bUseMaterialAttributes=true
  bUsedWithMeshDeformer=true
  bUsedWithSkeletalMesh=true
  blendMode=BLEND_Masked
  numCustomizedUVs=2
  opacityMaskClipValue=0.5
  pixelDepthOffsetMode=PDOM_Legacy
  twoSided=true
  usedShadingModels=Default Lit
 }
 nodes [
  n1 = Constant3Vector()
  n2 = Constant3Vector() {constant=<rgba 0.0613 0.0687 0.0938 1>}
  n3 = TextureSampleParameter2D(UVs=Multiply(A=TextureCoordinate(), B=Constant2Vector() {g=1, r=1})) {
   parameterName=DiffuseTexture
   samplerType=SAMPLERTYPE_Masks
   texture=/BRCosmetics/Gadgets/Assets/VinderTech_GliderChute/Glider_IgniteEgg/Textures/T_IgniteEgg_Spline_FX
  }
  n4 = Multiply(A=VertexColor().R, B=ScalarParameter() {defaultValue=7.5, parameterName=width})
  n5 = MaterialFunctionCall!(widthTip=@n4, Flatten=Constant(), widthBase=@n4) {materialFunction=/Game/Characters/Player/Male/Medium/Bodies/M_MED_Neon/Materials/MF_splineThickenM}
  n6 = SetMaterialAttributes!(
   MaterialAttributes
   Base Color=NamedRerouteDeclaration(None=LinearInterpolate(A=@n1, B=@n2, Alpha=@n3.RGB)) {name=baseColor, nodeColor=<rgba 0 1 0.2031 1>}
   Metallic=NamedRerouteDeclaration(None=ScalarParameter() {parameterName=Metallic}) {name=metallic, nodeColor=<rgba 0.9922 0 1 1>}
   Specular=NamedRerouteDeclaration(None=ScalarParameter() {defaultValue=0.1, parameterName=Specular}) {name=specular, nodeColor=<rgba 0.8984 0 1 1>}
   Roughness=NamedRerouteDeclaration(None=ScalarParameter() {defaultValue=1, parameterName=Roughness}) {name=roughness, nodeColor=<rgba 0.5 0 1 1>}
   Opacity Mask=MaterialFunctionCall() {materialFunction=&/MF_IgniteEgg_Glider}.OpacityMask
   Normal=NamedRerouteDeclaration(None=@n5.normal) {name=normal, nodeColor=<rgba 1 0.0703 0 1>}
   World Position Offset=NamedRerouteDeclaration(None=@n5.worldPositionOffset) {name=wpo, nodeColor=<rgba 0.125 0 1 1>}
  ) {attributeSetTypes [
    #1
    #2
    #3
    #4
    #5
    #6
    #7
   ]}
  MaterialFunctionCall(In=@n6) {materialFunction=&MF_TechArtOverrides_accessories}
  n7 = TextureCoordinate()
  n8 = ComponentMask(None=@n7) {r=true}
  n9 = ConstantBiasScale(None=Frac(None=@n8)) {bias=-0.5, scale=2}
  n10 = Constant()
  n11 = MaterialFunctionCall() {materialFunction=&TangentBasis}
  n12 = CrossProduct(A=Subtract(A=CameraPositionWS(), B=WorldPosition().XYZ), B=@n11.Y V)
  n13 = MaterialFunctionCall!(Vector1=@n12, Vector2=Multiply(A=@n11.Y V) {constB=-1}) {materialFunction=/Engine/Functions/Engine_MaterialFunctions02/Math/CreateThirdOrthogonalVector}
  MaterialFunctionCall(
   VectorToTransform=DeriveNormalZ(InXY=AppendVector(A=@n9, B=@n10))
   BasisX=Normalize(VectorInput=@n13.Vector1)
   BasisY=Normalize(VectorInput=@n13.Vector2)
   BasisZ=Normalize(VectorInput=@n13.Vector3)
  ) {materialFunction=&Transform3x3Matrix}
  n14 = ScalarParameter() {defaultValue=1, parameterName=inputWidthBase, sliderMin=10}
  n15 = ScalarParameter() {defaultValue=0.5, parameterName=inputWidthTip}
  n16 = ComponentMask(None=@n7) {g=true}
  n17 = Constant() {r=-0.5}
  Multiply(A=Multiply(A=LinearInterpolate(A=@n14, B=@n15, Alpha=@n16), B=Add(A=@n17, B=@n8)), B=Normalize(VectorInput=@n12))
 ]
}
