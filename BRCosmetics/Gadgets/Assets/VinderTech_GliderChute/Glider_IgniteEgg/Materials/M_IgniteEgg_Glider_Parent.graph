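# Parent masked, two-sided cloth material for the IgniteEgg glider, layering a noise-driven dust effect onto the 2023 anime character shader.
# The shared anime Pre, Calculations and Post functions supply the shading.
# MF_IgniteEgg_Glider supplies dust noise, masked noise and an opacity mask.
# Powered masked noise blends lit and shaded colors and scales DFL, a tinted dust color is built from noise, and the final Opacity Mask is overridden through reroutes.
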Material /BRCosmetics/Gadgets/Assets/VinderTech_GliderChute/Glider_IgniteEgg/Materials/M_IgniteEgg_Glider_Parent
graph {
 properties {
  bAutomaticallySetUsageInEditor=false
  bUseMaterialAttributes=true
  bUsedWithMeshDeformer=true
  bUsedWithMorphTargets=true
  bUsedWithSkeletalMesh=true
  bUsedWithStaticLighting=true
  blendMode=BLEND_Masked
  pixelDepthOffsetMode=PDOM_Legacy
  shadingModel=MSM_Cloth
  twoSided=true
  usedShadingModels=Cloth
 }
 nodes [
  n1 = NamedRerouteUsage()
  n2 = MaterialFunctionCall!(CustomUV) {materialFunction=/Game/Global/Materials/Character_Parent_2023/Anime_Parent_2023/MaterialFunctions/MFC_AnimeShader_2023_Pre}
  n3 = MaterialFunctionCall() {materialFunction=&/MF_IgniteEgg_Glider}
  n4 = Power(Base=@n3.MaskedNoise, Exp=ScalarParameter() {defaultValue=64, parameterName=DustColorOffset})
  LinearInterpolate(A=@n1, B=@n2.Color_Lit, Alpha=@n4)
  LinearInterpolate(A=Multiply(A=@n1) {constB=0.5}, B=@n2.Color_Shaded, Alpha=@n4)
  n5 = MaterialFunctionCall!(CustomLightVector, OverrideNormal_WS, STT=@n2.STT, Normal_Map=@n2.STT, DFL=LinearInterpolate(B=@n2.DFL, Alpha=@n4) {constB=0}, FX=@n2.STT) {materialFunction=/Game/Global/Materials/Character_Parent_2023/Anime_Parent_2023/MaterialFunctions/MFC_AnimeShader_2023_Calculations}
  n6 = GetMaterialAttributes(None=MaterialFunctionCall!(
    Shading_Gradient=@n5.Shading_Gradient
    Threshold_Gradient=@n5.Shading_Gradient
    Specular_Gradient=@n5.Shading_Gradient
    Sheen_Mask=@n5.Shading_Gradient
    Thickness=@n5.Shading_Gradient
    DFL_Mask=@n5.Shading_Gradient
    FX=@n5.Shading_Gradient
    Metallic_Gradient=@n5.Shading_Gradient
    RimLight_Gradient=@n5.Shading_Gradient
    NormalBuffer=@n5.Shading_Gradient
    --------
    CINE_LightSources=@n5.Shading_Gradient
    CINE_LightColor=@n5.Shading_Gradient
    ---------
    Color_Lit=@n2.Color_Lit
    Color_Shaded=@n2.Color_Lit
    SRM=@n2.Color_Lit
    Emissive=@n2.Color_Lit
    DFL_Color=@n2.Color_Lit
    Specular_Color=@n2.Color_Lit
    RimLight_Color=@n2.Color_Lit
   ) {materialFunction=/Game/Global/Materials/Character_Parent_2023/Anime_Parent_2023/MaterialFunctions/MFC_AnimeShader_2023_Post}.Result) {attributeGetTypes [#1]}
  NamedRerouteDeclaration(None=@n6.Opacity Mask) {name=InOpacityMask, nodeColor=<rgba 0 1 0.2734 1>}
  NamedRerouteDeclaration(None=@n3.OpacityMask) {name=OutOpacityMask, nodeColor=<rgba 0.9922 0 1 1>}
  NamedRerouteDeclaration(None=Multiply(
    A=VectorParameter() {parameterName=DustColor, defaultValue=<rgba 1 0.6157 0.5195 1>}.RGB
    B=LinearInterpolate(Alpha=@n3.DustNoise) {constA=0.2}
   )) {name=OutBaseColor, nodeColor=<rgba 1 0.7969 0 1>}
  SetMaterialAttributes!(MaterialAttributes=@n6.MaterialAttributes, Opacity Mask=NamedRerouteUsage()) {attributeSetTypes [#1]}
 ]
}
